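; Hookshot rope actor for the CoffeePuff ability, handling the rope's cosmetic visuals and affiliation-based cast and recall sounds.
; BeginPlay wiggles the rope and starts trail FX, and a knockback timeline drives that wiggle when a player is hooked.
; EndReel fires tighten FX and bends the rope, by delayed distance-scaled steps for hooked players or a pull-back timeline otherwise.
; Reel-in play rate derives from rope length and hotfixable pull speed.
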(event EventBeginPlay
  (bind _trail (Variables|Default|GetTrail))
  ;[0 1 0.08373 0.5] BeginPlay
  (|CallParentFunction)
  (Networking|CanExecuteCosmeticEvents
    (:True
      (CallFunction|PlaySoundwithPlayerAffiliation "/CoffeePuffGameplay/Assets/Sounds/IP_SC_CoffeePuff_Hookshot_Prj_Cast_Local.IP_SC_CoffeePuff_Hookshot_Prj_Cast_Local" "/CoffeePuffGameplay/Assets/Sounds/IP_SC_CoffeePuff_Hookshot_Prj_Cast_Friendly.IP_SC_CoffeePuff_Hookshot_Prj_Cast_Friendly" "/CoffeePuffGameplay/Assets/Sounds/IP_SC_CoffeePuff_Hookshot_Prj_Cast_Enemy.IP_SC_CoffeePuff_Hookshot_Prj_Cast_Enemy")
      (Effects|Components|ParticleSystem|SetBoolParameter _trail "EnableRopeTightenFX" false)
      (Components|Activation|Activate _trail true)
      (Rendering|Material|SetScalarParameterValueonMaterials (Variables|StaticMesh|GetRopeMesh) "WiggleAmount" 15.0)
      (Variables|Default|SetIs%20hooked%20a%20player false))))

(event Custom|Start%20Knockback%20Rope%20Tighten
  ;[0.1434 0.06398 0.15 0.5] Tighten the rope when attached to a player getting knocked back
  (Networking|CanExecuteCosmeticEvents
    (:True
      (Components|Timeline|SetPlayRate (Variables|B_CoffeePuff_Hookshot_Rope|GetKnockback%20Rope%20Tighten) (Math|Float|SafeDivide 1.0 (ScalableFloat|GetValueatLevel (Variables|Hotfix|GetHotfix_KnockbackDuration))))
      (Variables|Default|SetIs%20hooked%20a%20player true)
      (bind (_direction _tightenalpha) (Timeline|Knockback%20Rope%20Tighten Play)
        (:Update
          (Rendering|Material|SetScalarParameterValueonMaterials (Variables|StaticMesh|GetRopeMesh) "WiggleAmount" _tightenalpha))))))

(event EventEndReel
  (bind _returnvalue (Math|Float|MapRangeClamped (Math|Vector|Distance(Vector) (Variables|FortHookGunRope|GetStartLocation) (Variables|FortHookGunRope|GetEndLocation)) 0.0 3e+03 2e+01 3e+01))
  (bind _endlocation (Variables|FortHookGunRope|GetEndLocation))
  (bind _startlocation (Variables|FortHookGunRope|GetStartLocation))
  (bind _returnvalue_2 (Math|Vector|Distance(Vector) _startlocation _endlocation))
  (bind _trail (Variables|Default|GetTrail))
  (bind _ropemesh (Variables|StaticMesh|GetRopeMesh))
  ; Pull Rope
  (|CallParentFunction)
  (Networking|CanExecuteCosmeticEvents
    (:True
      (Audio|Components|Audio|FadeOut (Variables|Default|GetCastSound) 1.0 0.0 "Linear" :!class /Script/Engine.AudioComponent)
      (Effects|Components|ParticleSystem|SetFloatParameter _trail "RopeLength" _returnvalue_2)
      (Transformation|SetWorldRotation _trail (Math|Rotator|FindLookatRotation _startlocation _endlocation) false false)
      (Effects|Components|ParticleSystem|SetBoolParameter _trail "EnableRopeTightenFX" true)
      (Components|Activation|Activate _trail true)
      (CallFunction|PlaySoundwithPlayerAffiliation "/CoffeePuffGameplay/Assets/Sounds/IP_SC_CoffeePuff_Hookshot_Prj_Recall_Local.IP_SC_CoffeePuff_Hookshot_Prj_Recall_Local" "/CoffeePuffGameplay/Assets/Sounds/IP_SC_CoffeePuff_Hookshot_Prj_Recall_Friendly.IP_SC_CoffeePuff_Hookshot_Prj_Recall_Friendly" "/CoffeePuffGameplay/Assets/Sounds/IP_SC_CoffeePuff_Hookshot_Prj_Recall_Enemy.IP_SC_CoffeePuff_Hookshot_Prj_Recall_Enemy")
      (if (Variables|Default|GetIs%20hooked%20a%20player)
        (Utilities|FlowControl|Delay 0.07)
        (Rendering|Material|SetScalarParameterValueonMaterials _ropemesh "Bend" (* _returnvalue 0.75))
        (Utilities|FlowControl|Delay 0.08)
        (Rendering|Material|SetScalarParameterValueonMaterials _ropemesh "Bend" _returnvalue)
        (else
          (bind (_direction _bending) (Timeline|Timeline_PullingBack PlayFromStart)
            (:Update
              (Rendering|Material|SetScalarParameterValueonMaterials _ropemesh "Bend" (Math|Float|Lerp 0.0 5e+01 _bending))
              (Rendering|Material|SetScalarParameterValueonMaterials _ropemesh "BendBias" (Math|Float|Lerp 8e+01 1e+01 _bending)))))))))

(fn GetReelInPlayRate
  (return (Math|Float|SafeDivide 1.0 (Math|Float|Clamp(Float) (+ (Math|Float|SafeDivide (Math|Vector|Distance(Vector) (Variables|FortHookGunRope|GetEndLocation) (Variables|FortHookGunRope|GetStartLocation)) (ScalableFloat|GetValueatLevel (Variables|Hotfix|GetHotfix_HookshotPullSpeed))) 0.1) 0.3 (Variables|Settings|GetPullMontageSectionDuration)))))

(fn HideRopeProjectileMesh
  ;[0.1458 0.07621 0.15 0.5] Hide the projectile mesh.
  ; Called from the projectile actor after a successful hit.
  (Networking|CanExecuteCosmeticEvents
    (:True
      (Development|SetHiddeninGame (Variables|StaticMesh|GetRopeProjectileMesh) true false))))

(fn PlaySoundWithPlayerAffiliation
  (params (Sound :type Sound%20Base%20Object%20Reference) (Friendly :type Sound%20Base%20Object%20Reference) (Hostile :type Sound%20Base%20Object%20Reference))
  (bind _returnvalue (Game|GetInstigator))
  ; Determine Affiliation
  (if (Pawn|IsLocallyViewed _returnvalue)
    ;[0.0492 0.15 0.06017 0.5] Local Sound
    (bind _returnvalue_2 (Audio|SpawnSoundAttached Sound (Variables|StaticMesh|GetRopeProjectileMesh)))
    (Variables|Default|SetCastSound _returnvalue_2)
    (else
      ;[0.15 0.02311 0.03747 0.5] Non Local Sounds
      (bind _returnvalue_3 (Audio|SpawnSoundAttached (Utilities|Select :Index (Team|GetTeamAffiliationforActortoLocalPlayer _returnvalue) :Friendly Friendly :Hostile Hostile) (Variables|Default|GetAudio)))
      (Variables|Default|SetCastSound _returnvalue_3))))

(fn UpdateRopeProjectileMeshLocationAndRotation
  (params (ReelAlpha :type Float%20%28double-precision%29))
  (|CallParentFunction ReelAlpha))

(fn ConstructionScript
  (|CallParentFunction))
