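; Hookshot projectile for the CoffeePuff ability that delays its disable and reel-in when it hits a living pawn.
; With the custom-MME hotfix on, it flags knockback, optionally attaches to the victim, then calls parent disable once on pull-start event, death or timeout.
; The replicated knockback flag and rope OnReps tighten the rope and hide its projectile mesh.
; Attachment overrides use the bone nearest impact, and detaching stops projectile movement.
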(event EventDisableProjectile
  (params (HitResult :type Hit%20Result%20Structure))
  (bind (_bblockinghit _binitialoverlap _time _distance _location _impactpoint _normal _impactnormal _physmat _hitactor) (Collision|BreakHitResult (Variables|Transient|GetCachedDisableHitResult)))
  ;[0.2948 0.2137 0.02633 0.5] Disable Projectile: Overriden to adjust timing for pawn hits
  (Utilities|FlowControl|DoOnce false
    (:Completed
      (bind _output_get (Variables|Transient|SetCachedDisableHitResult HitResult))
      (if (ScalableFloat|GetValueasBool (Variables|Hotfix|GetHotfix_UsingCustomMME))
        ;;[0 0.15 0.1309 0.5] If pawn was hit, wait until knockback is finished and pull starts before disabling the projectile and starting reel-in
        (Utilities|IsValid (.HitActor (Collision|BreakHitResult _output_get))
          (:Is%20Valid
            (bind _asfort_pawn (Utilities|Casting|CastToFortPawn (.HitActor (Collision|BreakHitResult _output_get)))
              (:then
                (if (not (FortPawn|IsDead _asfort_pawn))
                  (Utilities|FlowControl|Sequence
                    (:then_0
                      (Variables|Knockback|SetbKnockbackStarted true)
                      (if (ScalableFloat|GetValueasBool (Variables|Hotfix|GetHotfix_AttachToHitLocation))
                        (bind _fortattachableprojectile (Variables|Default|GetFortAttachableProjectile)
                          (:then
                            (Gameplay|AuthorityAttachtoActor _fortattachableprojectile _output_get (.HitActor (Collision|BreakHitResult _output_get)) (ScalableFloat|GetValueatLevel (Variables|Hotfix|GetHotfix_AttachmentOffset)) (ScalableFloat|GetValueatLevel (Variables|Hotfix|GetHotfix_AttachmentOffset)) 0.5))
                          (:else))))
                    (:then_1
                      (bind _asynctaskproxy (Ability|Async|WaitGameplayEventToActor (.HitActor (Collision|BreakHitResult _output_get)) "(TagName=\"Event.Ability.CoffeePuff.MME.PullStart\")" true)
                        (:then
                          (Variables|Transient|Async|SetWaitForPullStartAction _asynctaskproxy)
                          (bind _asynctaskproxy_2 (Ability|Async|WaitGameplayEventToActor (.HitActor (Collision|BreakHitResult _output_get)) "(TagName=\"Event.Ability.CoffeePuff.MME.End\")" true)
                            (:then
                              (Variables|Transient|Async|SetWaitForMMEEndAction _asynctaskproxy_2))
                            (:EventReceived
                              _)))
                        (:EventReceived
                          _))
                      (Utilities|FlowControl|DoOnce false
                        (:Completed
                          (|CallParentFunction (Variables|Transient|GetCachedDisableHitResult))
                          (Utilities|Time|ClearandInvalidateTimerbyHandle (Variables|Transient|GetMMEPullTimeOutTimerHandle))
                          (Ability|Async|EndAction (Variables|Transient|Async|GetWaitForPullStartAction))
                          (Variables|Knockback|SetbKnockbackStarted false)
                          (bind _asfort_pawn_2 (Utilities|Casting|CastToFortPawn _hitactor))
                          (EventDispatchers|UnbindEventfromOnDied _asfort_pawn_2 (EventDispatchers|CreateEvent :!event Custom|HandlePullPawnDied)))))
                    (:then_2
                      (EventDispatchers|BindEventtoOnDied _asfort_pawn (delegate Custom|HandlePullPawnDied))
                      (bind _returnvalue (Utilities|Time|SetTimerbyEvent (delegate Custom|HandleMMEPullTimeOut) (+ (ScalableFloat|GetValueatLevel (Variables|Hotfix|GetHotfix_HookshotKnockbackPostDelay)) (ScalableFloat|GetValueatLevel (Variables|Hotfix|GetHotfix_HookshotKnockbackDuration)))))
                      (Variables|Transient|SetMMEPullTimeOutTimerHandle _returnvalue)))
                  (else
                    _)))
              (:CastFailed
                _)))
          (:Is%20Not%20Valid
            _))
        (else
          _)))))

(event Custom|HandlePullPawnDied
  (params (DamagedActor :type Actor%20Object%20Reference) (Damage :type Float%20%28single-precision%29) (InstigatedBy :type Controller%20Object%20Reference) (DamageCauser :type Actor%20Object%20Reference) (HitLocation :type Vector) (FHitComponent :type Primitive%20Component%20Object%20Reference) (BoneName :type Name) (Momentum :type Vector))
  (bind (_bblockinghit _binitialoverlap _time _distance _location _impactpoint _normal _impactnormal _physmat _hitactor) (Collision|BreakHitResult (Variables|Transient|GetCachedDisableHitResult)))
  ;[0.2948 0.2137 0.02633 0.5] Disable Projectile: Overriden to adjust timing for pawn hits
  ;;[0 0.15 0.1309 0.5] If pawn was hit, wait until knockback is finished and pull starts before disabling the projectile and starting reel-in
  (Utilities|FlowControl|DoOnce false
    (:Completed
      (|CallParentFunction (Variables|Transient|GetCachedDisableHitResult))
      (Utilities|Time|ClearandInvalidateTimerbyHandle (Variables|Transient|GetMMEPullTimeOutTimerHandle))
      (Ability|Async|EndAction (Variables|Transient|Async|GetWaitForPullStartAction))
      (Variables|Knockback|SetbKnockbackStarted false)
      (bind _asfort_pawn (Utilities|Casting|CastToFortPawn _hitactor))
      (EventDispatchers|UnbindEventfromOnDied _asfort_pawn (EventDispatchers|CreateEvent :!event Custom|HandlePullPawnDied)))))

(event Custom|HandleMMEPullTimeOut
  (bind (_bblockinghit _binitialoverlap _time _distance _location _impactpoint _normal _impactnormal _physmat _hitactor) (Collision|BreakHitResult (Variables|Transient|GetCachedDisableHitResult)))
  ;[0.2948 0.2137 0.02633 0.5] Disable Projectile: Overriden to adjust timing for pawn hits
  ;;[0 0.15 0.1309 0.5] If pawn was hit, wait until knockback is finished and pull starts before disabling the projectile and starting reel-in
  (Utilities|FlowControl|DoOnce false
    (:Completed
      (|CallParentFunction (Variables|Transient|GetCachedDisableHitResult))
      (Utilities|Time|ClearandInvalidateTimerbyHandle (Variables|Transient|GetMMEPullTimeOutTimerHandle))
      (Ability|Async|EndAction (Variables|Transient|Async|GetWaitForPullStartAction))
      (Variables|Knockback|SetbKnockbackStarted false)
      (bind _asfort_pawn (Utilities|Casting|CastToFortPawn _hitactor))
      (EventDispatchers|UnbindEventfromOnDied _asfort_pawn (EventDispatchers|CreateEvent :!event Custom|HandlePullPawnDied)))))

(event OnProjectileDetached(FortAttachableProjectile)
  (params (HostActor :type Actor%20Object%20Reference) (DetachReason :type EFortProjectileDetachReason%20Enum :default "HostDied"))
  (bind _projectilemovementcomponent (Variables|Movement|GetProjectileMovementComponent))
  ;[0.02561 0.08015 0.15 0.5] OnProjectileDetached: When the projectile is detached, the movement component is activated on the native side. We don't want this so deactivate it here.
  (Components|Movement|SetUpdatedComponent _projectilemovementcomponent 0)
  (Components|Activation|Deactivate _projectilemovementcomponent)
  (Projectile|SetVelocity _projectilemovementcomponent (0.0 0.0 0.0)))

(event EventEndPlay
  (params (EndPlayReason :type EEndPlayReason%20Enum :default "Destroyed"))
  (bind (_bblockinghit _binitialoverlap _time _distance _location _impactpoint _normal _impactnormal _physmat _hitactor) (Collision|BreakHitResult (Variables|Transient|GetCachedDisableHitResult)))
  ;[0.274 0.09664 0.02492 0.5] End Play: Clean up.
  (|CallParentFunction EndPlayReason)
  (Ability|Async|EndAction (Variables|Transient|Async|GetWaitForPullStartAction))
  (Networking|SwitchHasAuthority
    (:Authority
      (Variables|Knockback|SetbKnockbackStarted false)
      (Utilities|Time|ClearandInvalidateTimerbyHandle (Variables|Transient|GetMMEPullTimeOutTimerHandle))
      (bind _asfort_pawn (Utilities|Casting|CastToFortPawn _hitactor))
      (EventDispatchers|UnbindEventfromOnDied _asfort_pawn (EventDispatchers|CreateEvent :!event Custom|HandlePullPawnDied)))))

(event EventOnRep_Rope
  ;[0.01753 0.115 0.15 0.5] OnRep_Rope: Rope actor is replicated, if the actor has already started knockback, let the rope know.
  (|CallParentFunction)
  (Networking|CanExecuteCosmeticEvents
    (:True
      (if (Variables|Knockback|GetbKnockbackStarted)
        (bind _asb_coffee_puff_hookshot_rope (Utilities|Casting|CastToB_CoffeePuff_Hookshot_Rope (Variables|Transient|GetRope)))
        (Class|BCoffeePuffHookshotRope|StartKnockbackRopeTighten _asb_coffee_puff_hookshot_rope)
        (Class|BCoffeePuffHookshotRope|HideRopeProjectileMesh _asb_coffee_puff_hookshot_rope)))))

(event OnFinalizeAttachment(FortAttachableProjectile)
  (params (WorldLocation :type Vector%20%28by%20ref%29 :default "0, 0, 0") (WorldRotation :type Rotator%20%28by%20ref%29 :default "0, 0, 0") (SocketName :type Name%20%28by%20ref%29 :default "None"))
  (bind _fortattachableprojectile (Variables|Default|GetFortAttachableProjectile))
  (bind _returnvalue (Gameplay|GetAttachParamsHit _fortattachableprojectile))
  (bind (_bblockinghit _binitialoverlap _time _distance _location _impactpoint _normal _impactnormal _physmat _hitactor) (Collision|BreakHitResult _returnvalue))
  ;[0.15 0.03839 0.1384 0.5] OnFinalizeAttachment: Override the native attachment location and use the ImpactPoint directly.
  (bind _asfort_pawn (Utilities|Casting|CastToFortPawn _hitactor))
  (bind _mesh (Variables|/Script/Engine.Character|GetMesh _asfort_pawn)
    (:then
      (Gameplay|OverrideFinalizedAttachmentParams (Variables|Default|GetFortAttachableProjectile) (Components|SkinnedMesh|FindClosestBone _mesh _impactpoint 0.0 false) WorldRotation (Utilities|Name|MakeLiteralName "None")))
    (:else)))

(fn GetCurrentRopeLength
  (bind _length (|CallParentFunction))
  (return _length))

(fn OnRep_bKnockbackStarted
  ;[0.1469 0.07942 0.15 0.5] Tighten the rope when attached to a player getting knocked back
  (if (Variables|Knockback|GetbKnockbackStarted)
    (Rendering|SetVisibility (Variables|Default|GetStaticMesh) false false)
    (bind _asb_coffee_puff_hookshot_rope (Utilities|Casting|CastToB_CoffeePuff_Hookshot_Rope (Variables|Transient|GetRope)))
    (Class|BCoffeePuffHookshotRope|StartKnockbackRopeTighten _asb_coffee_puff_hookshot_rope)
    (Class|BCoffeePuffHookshotRope|HideRopeProjectileMesh _asb_coffee_puff_hookshot_rope)))

(fn ConstructionScript
  (|CallParentFunction))
